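# Post-process AnimBlueprint for the BlockStack Hope male character head, adjusting the pose copied from its attached parent mesh.
# Additive bone offsets on spine and chest feed a Control Rig, followed by cape and backpack attach-bone offsets.
# The EventGraph turns movement, weapon, hang-glider and mannequin state into rig inputs.
# In the front end, the Frontend_Idle_Bypass curve weights a per-bone DefaultSlot blend over pelvis, spine and leg branches.
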AnimBlueprint /BRCosmetics/Characters/Player/Male/Medium/Heads/M_MED_BlockStack_Hope_Head/Meshes/M_MED_BlockStack_Hope_Head_AnimBP_Post
target_skeleton=/BRCosmetics/Characters/Player/Male/Medium/Bodies/M_MED_BlockStack_Hope/Meshes/M_MED_BlockStack_Hope_Head_Skeleton
graphs [
 AnimGraph {nodes [
   n1 = "Transform (Modify) Bone - Bone: deform_spine_05"!(
    ComponentPose=Local To Component(LocalPose=Copy Pose From Mesh() {bCopyCurves=true, bUseAttachedParent=true})
    bAlphaBoolEnabled:bool=True
    Translation="0,0,0"
    Rotation="0.099,27.038,0.057"
    Scale="1,1,1"
   ) -> (Pose) {
    alphaInputType=Bool
    rotationMode=BMM_Additive
    rotationSpace=BCS_BoneSpace
    translationMode=BMM_Additive
    translationSpace=BCS_BoneSpace
    alphaBoolBlend {blendInTime=0.6, blendOutTime=0.2}
    boneToModify {boneName=deform_spine_05}
    rotation=<rot 0.099 27.038 0.057>
   }
   n2 = "Transform (Modify) Bone - Bone: chest_front"!(ComponentPose=@n1, bAlphaBoolEnabled:bool=True, Translation="-0.145,-1.737,0.182", Rotation="0,0,0", Scale="1,1,1") -> (Pose) {
    alphaInputType=Bool
    rotationMode=BMM_Additive
    rotationSpace=BCS_BoneSpace
    translationMode=BMM_Additive
    translationSpace=BCS_BoneSpace
    alphaBoolBlend {blendInTime=0.6, blendOutTime=0.2}
    boneToModify {boneName=chest_front}
    translation=<-0.145 -1.737 0.182>
   }
   n3 = Control Rig!(
    Source=Component To Local(ComponentPose=@n2)
    Alpha
    CrouchingBool:bool=Get bIsCrouching()
    StandingMovng2D:double=Get StandingMoving2D() -> (StandingMoving2D:double)
    CrouchingMoving2D:double=Get CrouchingMoving2D() -> (CrouchingMoving2D:double)
    Idle:double=Get Idle() -> (Idle:double)
    JumpingComplete:double=Get JumpingComplete() -> (JumpingComplete:double)
    OnlyGround:double=Get OnlyGround() -> (OnlyGround:double)
    GrindRail:bool=Get bIsOnGrindRail()
    HeadValue_Skydiving:double=Get HeadValue_Skydiving() -> (HeadValue_Skydiving:double)
    HeadValue_Crouching:double=Get HeadValue_Crouching() -> (HeadValue_Crouching:double)
    Skydiving:bool=Get bIsSkyDiving()
    pub isTargeting:bool=Get bIsTargeting()
    pub isAR Equipped:bool=Get isAR Equipped() -> (isAR Equipped:bool)
    is_OutLaw_SG:bool=Get Is_outLaw_SG() -> (Is_outLaw_SG:bool)
    SMG_P90:bool=Get SMG_SG() -> (SMG_SG:bool)
    pub isPistol Equipped:bool=Get isPistol Equipped() -> (isPistol Equipped:bool)
    pub PA Swing:bool=Get bIsPlayingMeleeAnim()
    HangGliding:double=Get HangGlidingAlpha() -> (HangGlidingAlpha:double)
    Emote:bool=Get bIsPlayingEmote()
    IsMannequin:bool="Get isMannequin "() -> ("isMannequin ":bool)
   ) -> (Pose) {controlRigAssetReference {controlRigAsset=&/M_MED_BlockStack_Hope_Head_CtrlRig}}
   n4 = "Transform (Modify) Bone - Bone: attach_cape"(ComponentPose=Local To Component(LocalPose=@n3), Alpha=1, Translation="0,-2,2.5", Rotation="0,0,0", Scale="1,1,1") {
    rotationMode=BMM_Additive
    rotationSpace=BCS_BoneSpace
    translationMode=BMM_Additive
    translationSpace=BCS_BoneSpace
    boneToModify {boneName=attach_cape}
    translation {y=-2, z=2.5}
   }
   n5 = "Transform (Modify) Bone - Bone: attach_cape"!(ComponentPose=@n4, bAlphaBoolEnabled:bool=True, Translation="0,-2,2.5", Rotation="0,0,0", Scale="1,1,1") -> (Pose) {
    alphaInputType=Bool
    rotationMode=BMM_Additive
    rotationSpace=BCS_BoneSpace
    translationMode=BMM_Additive
    translationSpace=BCS_BoneSpace
    alphaScaleBiasClamp {bInterpResult=true}
    boneToModify {boneName=attach_cape}
    translation {y=-2, z=2.5}
   }
   n6 = "Transform (Modify) Bone - Bone: attach_backpack"(ComponentPose=@n5, Alpha=1, Translation="0,0,0", Rotation="0,0,0", Scale="1,1,1") {
    rotationMode=BMM_Additive
    rotationSpace=BCS_BoneSpace
    translationMode=BMM_Additive
    translationSpace=BCS_BoneSpace
    alphaScaleBiasClamp {bInterpResult=true}
    boneToModify {boneName=attach_backpack}
   }
   n7 = "Transform (Modify) Bone - Bone: attach_backpack"(ComponentPose=@n6, Alpha=1, Translation="0,-0.8,0", Rotation="0,0,0", Scale="1,1,1") {
    rotationMode=BMM_Additive
    rotationSpace=BCS_BoneSpace
    translationMode=BMM_Additive
    translationSpace=BCS_BoneSpace
    alphaScaleBiasClamp {bInterpResult=true}
    boneToModify {boneName=attach_backpack}
    translation {y=-0.8}
   }
   n8 = "Transform (Modify) Bone - Bone: attach_backpack"!(ComponentPose=@n7, bAlphaBoolEnabled:bool=True, Translation="0,-2,0", Rotation="0,0,10", Scale="1,1,1") -> (Pose) {
    alphaInputType=Bool
    rotationMode=BMM_Additive
    translationMode=BMM_Additive
    translationSpace=BCS_BoneSpace
    alphaBoolBlend {blendInTime=0.2, blendOption=CubicInOut, blendOutTime=0.5}
    boneToModify {boneName=attach_backpack}
    rotation {roll=10}
    translation {y=-2}
   }
   n9 = Layered blend per bone(
    BasePose=Component To Local(ComponentPose=@n8)
    BlendPoses_0=Slot 'DefaultSlot'()
    BlendWeights_0=Get Custom Idle Is Playing Float() -> (Custom Idle Is Playing Float:double)
   ) {layerSetup [{branchFilters [
       {boneName=deform_pelvis}
       {boneName=deform_spine_02}
       {boneName=deform_thigh_l}
       {boneName=deform_thighUp_l}
       {boneName=deform_thigh_bk_l}
       {boneName=deform_thigh_fr_l}
       {boneName=deform_thigh_r}
       {boneName=deform_thighUp_r}
       {boneName=deform_thigh_bk_r}
       {boneName=deform_thigh_fr_r}
       {boneName=calf_twist_01_l}
       {boneName=calf_twist_02_l}
       {boneName=deform_foot_l}
       {boneName=deform_ball_l}
       {boneName=foot_bk_l}
       {boneName=foot_fr_l}
       {boneName=deform_knee_in_base_l}
       {boneName=deform_knee_in_l}
       {boneName=deform_knee_out_base_l}
       {boneName=deform_knee_out_l}
       {boneName=calf_twist_01_r}
       {boneName=calf_twist_02_r}
       {boneName=deform_foot_r}
       {boneName=deform_ball_r}
       {boneName=foot_bk_r}
       {boneName=foot_fr_r}
       {boneName=deform_knee_in_base_r}
       {boneName=deform_knee_in_r}
       {boneName=deform_knee_out_base_r}
       {boneName=deform_knee_out_r}
       {boneName=thigh_twist_01_r}
      ]}]}
   Output Pose(Result=@n9)
  ]}
 EventGraph {nodes [
   Event BlueprintUpdateAnimation()
   > n1 = Sequence!(execute) -> (then_0, then_1, then_2:Exec, then_3:Exec, then_4:Exec)
   n2 = Get bIsCrouching()
   n3 = Get bIsMoving2D()
   > Set CrouchingMoving2D(CrouchingMoving2D:double="To Float (Boolean)"(InBool=AND Boolean(A=@n2, B=@n3))) -> (Output_Get:double)
   n4 = NOT Boolean(A=Get bIsCrouching())
   n5 = Get bIsMoving2D()
   > Set StandingMoving2D(StandingMoving2D:double="To Float (Boolean)"(InBool=AND Boolean(A=@n5, B=@n4))) -> (Output_Get:double)
   n6 = NOT Boolean(A=Get bIsPlayingEmote())
   n7 = NOT Boolean(A=Get bIsMoving2D())
   n8 = NOT Boolean(A=Get bIsSkyDiving())
   > Set Idle(Idle:double="To Float (Boolean)"(InBool=AND Boolean!(A=@n7, B=@n8, C:bool=@n6) -> (ReturnValue))) -> (Output_Get:double)
   n9 = AND Boolean!(A=NOT Boolean(A=Get bIsSkyDiving()), B=Get bIsFalling()) -> (ReturnValue)
   n10 = Get bIsJumping()
   > Set JumpingComplete(JumpingComplete:double="To Float (Boolean)"(InBool=OR Boolean(A=@n9, B=@n10))) -> (Output_Get:double)
   n11 = NOT Boolean(A=Get bIsSkyDiving())
   n12 = NOT Boolean(A=Get bIsPlayingEmote())
   n13 = NOT Boolean(A=Get bIsSwimSprinting() -> (bIsSwimSprinting:bool))
   n14 = NOT Boolean(A=Get bIsPlayingMeleeAnim())
   > Set OnlyGround(OnlyGround:double="To Float (Boolean)"(InBool=AND Boolean!(A=@n11, B=@n13, C:bool=@n14, D:bool=@n12) -> (ReturnValue))) -> (Output_Get:double)
   n15 = Get bIsCrouching()
   > Set Standing(Standing:double="To Float (Boolean)"(InBool=NOT Boolean(A=@n15))) -> (Output_Get:double)
   n16 = Get bIsTacticalSprinting()
   n17 = Get bIsJumping()
   > Set TaciticalSprintingJump(TaciticalSprintingJump:bool=OR Boolean(A=AND Boolean!(A=@n16, B=@n17) -> (ReturnValue), B=Get bIsJumping())) -> (Output_Get:bool)
   n18 = Branch(execute=@n1.then_1, Condition=Get bIsSkyDiving())
   > Set HeadValue_Skydiving(HeadValue_Skydiving:double=0) -> (Output_Get:double)
   Set HeadValue_Skydiving(execute=@n18.else, HeadValue_Skydiving=0.4)
   n19 = Branch(execute=@n1.then_2, Condition=Get bIsCrouching())
   > Set HeadValue_Crouching(HeadValue_Crouching:double=0.3) -> (Output_Get:double)
   Set HeadValue_Crouching(execute=@n19.else, HeadValue_Crouching=0.7)
   n20 = GetOwningActor()
   Cast To FortPlayerMannequin!(execute=@n1.then_4, Object=@n20) -> (then, CastFailed, AsFort Player Mannequin:Fort Player Mannequin Object Reference) {is=title K2Node_DynamicCast}
   > n21 = For Each Loop(Array:Array of Skeletal Mesh Component Object References=Get Components By Class!(self:Actor Object Reference=@n20, ComponentClass:Actor Component Class Reference=/Script/Engine.SkeletalMeshComponent) -> (ReturnValue:Array of Skeletal Mesh Component Object References)) -> (Array Element:Skeletal Mesh Component Object Reference)
   > n22 = Branch(Condition="Equal (Object)"(
     A=Get SkeletalMesh!(self:Skinned Mesh Component Object Reference=@n21) -> (SkeletalMesh:Skeletal Mesh Object Reference)
     B=/BRCosmetics/Characters/Player/Male/Medium/Bodies/M_MED_BlockStack_Hope/Meshes/M_MED_BlockStack_Hope_Head
    ))
   > "Set isMannequin "("isMannequin ":bool=false) -> (Output_Get:bool)
   "Set isMannequin "(execute=@n22.else, "isMannequin "=true)
   Branch(execute=@n1.then_3, Condition=Get bIsInFrontEnd())
   > Set Custom Idle Is Playing Float(Custom Idle Is Playing Float:double=GetCurveValue(self=GetOwnerFrontEndAnimInstance(), CurveName=Frontend_Idle_Bypass)) -> (Output_Get:double)
   Event OnParachuteOpened()
   > Switch on EGliderType(Selection=Get PlayerGliderType() -> (PlayerGliderType=HangGlider))
   > Set HangGlidingAlpha(HangGlidingAlpha:double=1) -> (Output_Get:double)
   Event OnParachuteClosed()
   > Set HangGlidingAlpha(HangGlidingAlpha=0)
   Event OnWeaponChanged()
   > Set isAR Equipped(isAR Equipped:bool=false) -> (Output_Get:bool)
   > Set isPistol Equipped(isPistol Equipped:bool=false) -> (Output_Get:bool)
   > Set Is_outLaw_SG(Is_outLaw_SG:bool=false) -> (Output_Get:bool)
   > Set SMG_SG(SMG_SG:bool=false) -> (Output_Get:bool)
   > n23 = Switch on EFortWeaponCoreAnimation!(execute, Selection:EFortWeaponCoreAnimation Enum=Get WeaponCoreAnim!() -> (WeaponCoreAnim)) -> (
    Melee:Exec
    Pistol:Exec
    Shotgun:Exec
    PaperBlueprint:Exec
    Rifle:Exec
    MeleeOneHand:Exec
    MachinePistol:Exec
    RocketLauncher:Exec
    GrenadeLauncher:Exec
    GoingCommando:Exec
    AssaultRifle:Exec
    TacticalShotgun:Exec
    SniperRifle:Exec
    TrapPlacement:Exec
    ShoulderLauncher:Exec
    AbilityDecoTool:Exec
    Crossbow:Exec
    C4:Exec
    RemoteControl:Exec
    DualWield:Exec
    AR_BullPup:Exec
    AR_ForwardGrip:Exec
    MedPackPaddles:Exec
    SMG_P90:Exec
    AR_DrumGun:Exec
    Consumable_Small:Exec
    Consumable_Large:Exec
    Balloon:Exec
    MountedTurret:Exec
    CreativeTool:Exec
    ExplosiveBow:Exec
    AshtonIndigo:Exec
    AshtonChicago:Exec
    MeleeDualWield:Exec
    Unarmed:Exec
   )
   Set isAR Equipped(execute=@n23.AssaultRifle, isAR Equipped=true)
   Set isPistol Equipped(execute=@n23.Pistol, isPistol Equipped=true)
   Set SMG_SG(execute=@n23.SMG_P90, SMG_SG=true)
   Set Is_outLaw_SG(execute=@n23.TacticalShotgun, Is_outLaw_SG=true)
  ]}
]
variables [
 CrouchingMoving2D {type=number}
 Custom Idle Is Playing Float {type=number}
 HangGlidingAlpha {type=number}
 HeadValue_Crouching {type=number}
 HeadValue_Skydiving {type=number}
 Idle {type=number, default=1}
 Is_outLaw_SG {type=boolean}
 isAR Equipped {type=boolean}
 "isMannequin " {type=boolean}
 isPistol Equipped {type=boolean}
 JumpingComplete {type=number}
 OnlyGround {type=number}
 SMG_SG {type=boolean}
 Standing {type=number}
 StandingMoving2D {type=number}
 TaciticalSprintingJump {type=boolean}
]
